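# Translucent unlit test material for toon fire on Niagara sprite, ribbon and mesh particles.
# A sine of the tiled V coordinate blends orange and red into repeating color bands.
# The red channel of a looping toon fire texture parameter, multiplied by particle alpha, forms the opacity mask.
# A quality switch applies depth fade to that mask on Default, High and Epic, skipping it on Low and Medium.
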Material /WildEstateTech/FX/ToonFire/M_Test_ToonFire
graph {
 properties {
  bAutomaticallySetUsageInEditor=false
  bUsedWithNiagaraMeshParticles=true
  bUsedWithNiagaraRibbons=true
  bUsedWithNiagaraSprites=true
  blendMode=BLEND_Translucent
  shadingModel=MSM_Unlit
  usedShadingModels=Unlit
 }
 nodes [
  n1 = TextureCoordinate() {vTiling=4.8}
  LinearInterpolate(
   A=Constant3Vector() {constant=<rgba 1 0.221 0 1>}
   B=Constant3Vector() {constant=<rgba 1 0 0.015 1>}
   Alpha=Sine(None=ComponentMask(None=@n1) {g=true})
  )
  Multiply(A=TextureCoordinate()) {constB=5}
  n2 = Multiply(
   A=ParticleColor().A
   B=TextureSampleParameter2D() {parameterName=Texture, texture=&/T_WildEstate_LoopingToonFire_Sharp_061025}.R
  )
  n3 = DepthFade(Opacity=@n2) {fadeDistanceDefault=20}
  QualitySwitch(Default=@n3, Low=@n2, High=@n3, Medium=@n2, Epic=@n3)
 ]
}
